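; Base Blueprint for time-of-day volumes whose DaySequenceModifier component alters the day sequence, toggled by UseVolume singly or map-wide.
; On cosmetic-capable instances, setup binds the modifier to the Time of Day Manager, found asynchronously and re-resolved in editor on actor changes.
; The player controller, polled per tick until valid, becomes the blend target; replay scrubbing reruns setup.
; Empty UpdateScalabilitySequence and OnBlendTargetSet functions serve as hooks.
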(event Custom|Setup%20TODM
  (bind _outputdelegate (EventDispatchers|CreateEvent :!event UpdateScalabilitySequence))
  (bind _returnvalue (WorldSubsystems|GetDaySequenceSubsystem))
  (bind _daysequencemodifier (Variables|Default|GetDaySequenceModifier))
  (if (Replay|GetIsPlayingReplay)
    (Variables|Default|SetWaitForTimeOfDayManagerTask 0)
    (else
      _))
  (Utilities|FlowControl|Sequence
    (:then_0
      ;[1 1 1] Use event dispatcher here to ensure that these will fire in editor as a result of the CallInEditor setting on UpdateScalabilitySequences
      (EventDispatchers|BindEventtoOnPostEnableModifier _daysequencemodifier _outputdelegate)
      (EventDispatchers|BindEventtoOnPostReinitializeSubSequences _daysequencemodifier _outputdelegate))
    (:then_1
      (Variables|Default|GetWaitForTimeOfDayManagerTask
        (:then)
        (:else
          (bind (_asynctask _todm) (Default|GetTODM self)
            (:OnAsyncTaskActivated
              (Variables|Default|SetWaitForTimeOfDayManagerTask _asynctask))
            (:Out
              (CallFunction|InitModifier _todm)
              (if (Variables|Default|GetUseVolume)
                (CallFunction|EnableDaysequenceVolume)
                (else
                  (CallFunction|DisableDaysequenceVolume)))
              (Variables|Default|SetTODM%20Async%20Fired true)
              (CallFunction|PostTODMCached))))) ; This will always be invalid in editor as we don't rely on the async task
    )
    (:then_2
      ;[1 1 1] Listen for TODM changes while in editor (we no longer rely on async actions in editor, only in game)
      (if (Editor|InEditor)
        (DaySequence|UnbindEventfromOnDaySequenceActorSet _returnvalue (delegate Custom|OnDaySequenceActorSet))
        (DaySequence|AssignOnDaySequenceActorSet _returnvalue (delegate Custom|OnDaySequenceActorSet))))))

(event Custom|OnDaySequenceActorSet
  (params (NewActor :type Day%20Sequence%20Actor%20Object%20Reference))
  ;[1 1 1] Listen for TODM changes while in editor (we no longer rely on async actions in editor, only in game)
  (CallFunction|SetupTODM))

(event EventBeginPlay
  (|CallParentFunction)
  (Networking|CanExecuteCosmeticEvents
    (:True
      (CallFunction|SetupTODM) ; Will call Post TODM Cached when TODM is updated.
    )
    (:False
      (Collision|SetActorEnableCollision self false))))

(event Custom|Post%20TODM%20Cached
  (CallFunction|SetupControllerEvents))

(event OnPostEnableModifier(DaySequenceModifier)
  ;[1 1 1] Post load cleanup
  (if (Variables|Default|GetUseVolume)
    (CallFunction|EnableDaysequenceVolume)
    (else
      (CallFunction|DisableDaysequenceVolume))))

(event EventEndPlay
  (params (EndPlayReason :type EEndPlayReason%20Enum :default "Destroyed"))
  (|CallParentFunction EndPlayReason)
  (DaySequence|UnbindfromDaySequenceActor (Variables|Default|GetDaySequenceModifier) :!class /Script/DaySequence.DaySequenceModifierComponent))

(event Custom|EnableDaysequenceVolume
  ;[1 1 1] Enable
  (Variables|Default|SetUseVolume true)
  (DaySequence|EnableComponent (Variables|Default|GetDaySequenceModifier)))

(event Custom|DisableDaysequenceVolume
  ;[1 1 1] Disable
  (Variables|Default|SetUseVolume false)
  (DaySequence|DisableComponent (Variables|Default|GetDaySequenceModifier)))

(event Custom|EnableAllDaysequenceVolumesInMap
  ;[1 1 1] Enable
  (bind _outactors (Actor|GetAllActorsOfClass "/Game/TimeOfDay/TODM_Sequence/Blueprint/Volume/B_DaySequenceVolume_Parent.B_DaySequenceVolume_Parent_C"))
  (for _array_element _outactors
    (CallFunction|EnableDaysequenceVolume _array_element)))

(event Custom|DisableAllDaysequenceVolumesInMap
  ;[1 1 1] Disable
  (bind _outactors (Actor|GetAllActorsOfClass "/Game/TimeOfDay/TODM_Sequence/Blueprint/Volume/B_DaySequenceVolume_Parent.B_DaySequenceVolume_Parent_C"))
  (for _array_element _outactors
    (CallFunction|DisableDaysequenceVolume _array_element)))

(event Custom|Setup%20Controller%20Events
  (bind _returnvalue (Game|GetPlayerController 0))
  ;[1 1 1] PlayerController/ViewTarget Events
  ;;[1 1 1] Wait for PlayerController to be valid
  (Utilities|IsValid _returnvalue
    (:Is%20Valid
      _)
    (:Is%20Not%20Valid
      (Utilities|FlowControl|DelayUntilNextTick)
      (Variables|Default|SetPlayerControllerWaitTickCount (+ 1 (Variables|Default|GetPlayerControllerWaitTickCount))) ; We don't want to log spam here\nbut we track this for later inspection
    ))
  (bind _asfort_replay_spectator (Utilities|Casting|CastToFortReplaySpectator _returnvalue)
    (:then
      (CallFunction|HandleSpectatorController _asfort_replay_spectator))
    (:CastFailed
      _)) ; Divert for replay
  (CallFunction|SetBlendTarget _returnvalue))

(event Custom|DaySeqVolDebugDump
  ;[1 1 1] Debug Events
  (Utilities|FlowControl|Sequence
    (:then_0
      (Fortnite|FortShippingLog self (Utilities|String|Append "CurrentViewTarget: " (Utilities|GetDisplayName (Variables|Default|GetCurrentViewTarget)))))
    (:then_1
      (Fortnite|FortShippingLog self (Utilities|String|Append "Game State Async Fired: " (Utilities|String|ToString(Boolean) (Variables|Default|GetGame%20State%20Async%20Fired)))))
    (:then_2
      (Fortnite|FortShippingLog self (Utilities|String|Append "TODM Async Fired: " (Utilities|String|ToString(Boolean) (Variables|Default|GetTODM%20Async%20Fired)))))
    (:then_3
      (Fortnite|FortShippingLog self (Utilities|String|Append "Ticks spent waiting for PlayerController: " (Utilities|String|ToString(Integer) (Variables|Default|GetPlayerControllerWaitTickCount)))))))

(event Custom|Handle%20Spectator%20Controller
  (params (InSpectatorController :type Fort%20Replay%20Spectator%20Object%20Reference))
  ;[1 1 1] PlayerController/ViewTarget Events
  ;;[1 1 1] Replay Path
  (EventDispatchers|BindEventtoOnTimelineScrubbed InSpectatorController (EventDispatchers|CreateEvent :!event Custom|Setup%20TODM)))

(fn InitModifier
  (params (TODM%20Inteface :type Fort%20Time%20Of%20Day%20Manager%20Interface%20Interface))
  (bind _asfort_day_sequence_actor (Utilities|Casting|CastToFortDaySequenceActor TODM%20Inteface))
  (if (not (== _asfort_day_sequence_actor (Variables|Default|GetCachedTimeOfDayManager)))
    (bind _output_get (Variables|Default|SetCachedTimeOfDayManager _asfort_day_sequence_actor))
    (DaySequence|BindtoDaySequenceActor (Variables|Default|GetDaySequenceModifier) _output_get :!class /Script/DaySequence.DaySequenceModifierComponent)))

(fn OnBlendTargetSet)

(fn SetBlendTarget
  (params (InPC :type Player%20Controller%20Object%20Reference))
  (bind _daysequencemodifier (Variables|Default|GetDaySequenceModifier))
  (if (!= InPC (Variables|Default|GetCurrentViewTarget))
    (Utilities|FlowControl|Sequence
      (:then_0
        (bind _output_get (Variables|Default|SetCurrentViewTarget InPC)) ; Set ViewTarget
        (Utilities|IsValid _output_get
          (:Is%20Not%20Valid
            (Fortnite|FortShippingLog self "SetBlendTarget called with invalid view target!" true))))
      (:then_1
        (DaySequence|SetBlendTarget _daysequencemodifier (Variables|Default|GetCurrentViewTarget))
        (Variables|/Script/DaySequence.DaySequenceModifierComponent|SetUserBlendWeight 1.0 _daysequencemodifier))
      (:then_2
        (CallFunction|OnBlendTargetSet)))))

(fn UpdateScalabilitySequence)

(fn ConstructionScript
  (|CallParentFunction)
  (Networking|CanExecuteCosmeticEvents
    (:True
      (Utilities|FlowControl|Sequence
        (:then_0
          (CallFunction|SetupTODM))
        (:then_1
          (if (Variables|Default|GetUseVolume)
            (CallFunction|EnableDaysequenceVolume)
            (else
              (CallFunction|DisableDaysequenceVolume))))))))
